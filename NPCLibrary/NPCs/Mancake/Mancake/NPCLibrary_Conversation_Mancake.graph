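# Conversation database for the Mancake NPC, providing his greeting dialogue and access to NPC quests and services.
# The Mancake entry tag leads to a Speech node with a default line plus contextual lines.
# Contextual lines key off a first-ever-conversation check, prioritized through a curve row, or the player's equipped character cosmetic.
# The speech then branches through link nodes into the shared NPCLibrary Quests and BasicServices conversation entries.
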ConversationDatabase /NPCLibrary/NPCs/Mancake/Mancake/NPCLibrary_Conversation_Mancake
entry_points [{tag=Conversation.Entry.Mancake, destinations [#1]}]
nodes [
 ConversationEntryPointNode_7 {
  id=#1
  class=ConversationEntryPointNode
  outputs [#2]
  properties {entryTag {tagName=Conversation.Entry.Mancake}}
 }
 Speech_C_4 {
  id=#2
  class=Speech_C
  outputs [#3, #4]
  properties {
   nodeName=""
   generalConfig {
    defaultMessage=Someone broke the popcorn machine. Was it you?
    contextualMessages [
     {
      message="Every time I get invested in a movie, a bunch of you show up and start brawling."
      contextRequirements [{requirement {class=FortControllerRequirement_IsFirstEverConversationWithNPC}}]
      priority {curve {rowName=Tandem.Speech.Priority.FirstEverInteraction}}
     }
     {
      message="What brings you to these parts, pardner?"
      requirementMatchPolicy=RequireAny
      contextRequirements [{requirement {
         class=FortControllerRequirement_HasCID
         properties {allowedCIds [
           /BRCosmetics/Athena/Items/Cosmetics/Characters/CID_577_Athena_Commando_F_CODSquadPlaid
           /BRCosmetics/Athena/Items/Cosmetics/Characters/CID_348_Athena_Commando_F_Medusa
           /BRCosmetics/Athena/Items/Cosmetics/Characters/CID_498_Athena_Commando_M_WildWest
           /BRCosmetics/Athena/Items/Cosmetics/Characters/CID_497_Athena_Commando_F_WildWest
           /BRCosmetics/Athena/Items/Cosmetics/Characters/CID_237_Athena_Commando_F_Cowgirl
           /BRCosmetics/Athena/Items/Cosmetics/Characters/CID_250_Athena_Commando_M_EvilCowboy
          ]}
        }}]
      priority {value=0, curve {curveTable=null, rowName=None}}
     }
     {
      message="Why don't ya flap off, short stack?"
      contextRequirements [{requirement {class=FortControllerRequirement_HasCID, properties {allowedCIds [&Character_Bites]}}}]
      priority {value=0, curve {curveTable=null, rowName=None}}
     }
     {
      message="You're more than you seem, ain't cha?"
      contextRequirements [{requirement {class=FortControllerRequirement_HasCID, properties {allowedCIds [&Character_PinkSpike]}}}]
      priority {value=0, curve {curveTable=null, rowName=None}}
     }
     {
      message="You walk quiet and carry some big claws, dontcha?"
      contextRequirements [{requirement {class=FortControllerRequirement_HasCID, properties {allowedCIds [&Character_BadBear]}}}]
      priority {value=0, curve {curveTable=null, rowName=None}}
     }
     {
      message="You're all butter, no sizzle."
      contextRequirements [{requirement {class=FortControllerRequirement_HasCID, properties {allowedCIds [&Character_TheHerald]}}}]
      priority {value=0, curve {curveTable=null, rowName=None}}
     }
     {
      message="Spider-Cake, Spider-Cake..."
      contextRequirements [{requirement {class=FortControllerRequirement_HasCID, properties {allowedCIds [&Character_RoseDust]}}}]
      priority {value=0, curve {curveTable=null, rowName=None}}
     }
    ]
   }
  }
 }
 ConversationLinkNode_0 {id=#4, class=ConversationLinkNode, properties {remoteEntryTag {tagName=Conversation.Entry.BasicServices}}}
 ConversationLinkNode_1 {id=#3, class=ConversationLinkNode, properties {remoteEntryTag {tagName=Conversation.Entry.NPCLibrary.Quests}}}
]
